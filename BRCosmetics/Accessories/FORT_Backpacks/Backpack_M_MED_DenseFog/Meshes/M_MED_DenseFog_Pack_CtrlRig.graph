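# Control Rig for the M_MED_DenseFog backpack that poses its five three-bone dangling "breads" chains.
# End controls are parent-constrained between themselves and the dyn_breadsmain bone, weighted toward the bone, then drive two-bone IK with mid controls as pole vectors.
# When isSplineSmoothEnabled is on, SplineChainSmooth refits each chain, stretched, to a spline through its joints.
# The upper02 chain takes its pole vector from subbread02's mid control, not its own.
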ControlRigBlueprint /BRCosmetics/Accessories/FORT_Backpacks/Backpack_M_MED_DenseFog/Meshes/M_MED_DenseFog_Pack_CtrlRig
graphs [
 New Function_ContainedGraph {nodes [
   n1 = RigVMFunctionEntryNode()
   > DISPATCH_RigVMDispatch_ArrayReset = RigVMDispatchNode(Array=RigVMVariableNode() {variable=TempPositionArray}.Value) {template="DISPATCH_RigVMDispatch_ArrayReset(io Array)"}
   > n2 = RigVMDispatchNode(Array=@n1.ThreeJoints) {template="DISPATCH_RigVMDispatch_ArrayIterator(in Array,out Element,out Index,out Count,out Ratio)"}
   > DISPATCH_RigVMDispatch_ArrayAdd = RigVMDispatchNode(
    Element=&RigUnit_GetTransform(bInitial=False, Item=@n2.Element).Transform.Translation
    Array=RigVMVariableNode() {variable=TempPositionArray}.Value
   ) {template="DISPATCH_RigVMDispatch_ArrayAdd(io Array,in Element,out Index)"}
   n3 = RigVMVariableNode() {variable=TempPositionArray}
   DISPATCH_RigVMDispatch_ArrayGetAtIndex = RigVMDispatchNode(Index=0, Array=@n3.Value) {template="DISPATCH_RigVMDispatch_ArrayGetAtIndex(in Array,in Index,out Element)"}
   SplineThrough3Points = RigVMFunctionReferenceNode(
    t=0
    Point 0=@DISPATCH_RigVMDispatch_ArrayGetAtIndex.Element
    Point 1=RigVMDispatchNode(Index=1, Array=@n3.Value) {template="DISPATCH_RigVMDispatch_ArrayGetAtIndex(in Array,in Index,out Element)"}.Element
    Point 2=RigVMDispatchNode(Index=2, Array=@n3.Value) {template="DISPATCH_RigVMDispatch_ArrayGetAtIndex(in Array,in Index,out Element)"}.Element
    Execute=@n2.Completed
   )
   RigVMFunction_ControlFlowBranch = &RigVMFunction_ControlFlowBranch(Condition=@n1.DebugThickness, ExecuteContext=@SplineThrough3Points.Execute)
   /Script/ControlRigSpline.RigUnit_DrawControlRigSpline(Color=(R=1,G=0,B=0,A=1), Detail=16, Thickness=1, ExecutePin=@RigVMFunction_ControlFlowBranch.True, Spline=@SplineThrough3Points.Spline) {is=struct RigVMUnitNode}
   /Script/ControlRigSpline.RigUnit_FitChainToSplineCurveItemArray(
    Alignment=Stretched
    DebugSettings=(bEnabled=False,Scale=1,CurveColor=(R=1,G=1,B=0,A=1),SegmentsColor=(R=1,G=0,B=0,A=1),WorldOffset=(Rotation=(X=0,Y=0,Z=0,W=1),Translation=(X=0,Y=0,Z=0),Scale3D=(X=1,Y=1,Z=1)))
    Maximum=1
    Minimum=0
    PoleVectorPosition=(X=0,Y=0,Z=0)
    PrimaryAxis=(X=1,Y=0,Z=0)
    RotationEaseType=QuadraticEaseInOut
    Rotations=()
    SamplingPrecision=12
    SecondaryAxis=(X=0,Y=0,Z=0)
    Weight=1
    bPropagateToChildren=false
    Spline=@SplineThrough3Points.Spline
    Items=RigVMVariableNode() {variable=ThreeJoints}.Value
    ExecutePin=@RigVMFunction_ControlFlowBranch.Completed
   ) {is=struct RigVMUnitNode}
   > RigVMFunctionReturnNode()
  ]}
 RigVMFunctionLibrary {nodes [SplineChainSmooth = RigVMCollapseNode(ExecuteContext=(), ThreeJoints=())]}
 RigVMModel {nodes [
   n1 = &RigUnit_ItemArray(Items=((Type=Control,Name="dyn_breads_side_03_ctrl"),(Type=Control,Name="dyn_subbread_upper02_03_ctrl"),(Type=Control,Name="dyn_subbread_upper01_03_ctrl"),(Type=Control,Name="dyn_subbread01_03_ctrl"),(Type=Control,Name="dyn_subbread02_03_ctrl")))
   DISPATCH_RigVMDispatch_ArrayIterator = RigVMDispatchNode(ExecuteContext=(), Array=@n1.Items) {template="DISPATCH_RigVMDispatch_ArrayIterator(in Array,out Element,out Index,out Count,out Ratio)"}
   > &RigUnit_ParentConstraint(
    AdvancedSettings=(InterpolationType=Shortest,RotationOrderForFilter=XZY)
    Parents=((Item=(Type=Bone,Name="None"),Weight=0.25),(Item=(Type=Bone,Name="dyn_breadsmain"),Weight=0.75))
    Weight=1
    bMaintainOffset=true
    Child=@DISPATCH_RigVMDispatch_ArrayIterator.Element
    Parents.0.Item=@DISPATCH_RigVMDispatch_ArrayIterator.Element
   )
   /Script/ControlRig.RigUnit_TwoBoneIKSimplePerItem(
    DebugSettings=(bEnabled=False,Scale=10,WorldOffset=(Rotation=(X=0,Y=0,Z=0,W=1),Translation=(X=0,Y=0,Z=0),Scale3D=(X=1,Y=1,Z=1)))
    EffectorItem=(Type=Bone,Name="dyn_subbread01_03")
    ItemA=(Type=Bone,Name="dyn_subbread01_01")
    ItemALength=0
    ItemB=(Type=Bone,Name="dyn_subbread01_02")
    ItemBLength=0
    PoleVectorKind=Location
    PoleVectorSpace=(Type=Bone,Name="None")
    PrimaryAxis=(X=1,Y=0,Z=0)
    SecondaryAxis=(X=0,Y=-1,Z=0)
    SecondaryAxisWeight=0
    StretchMaximumRatio=1.25
    StretchStartRatio=0.75
    Weight=1
    bEnableStretch=false
    bPropagateToChildren=false
    Effector=&RigUnit_GetTransform(Item=(Type=Control,Name="dyn_subbread01_03_ctrl"), bInitial=False).Transform
    PoleVector=&RigUnit_GetTransform(Item=(Type=Control,Name="dyn_subbread01_02_ctrl"), bInitial=False).Transform.Translation
    ExecutePin=@DISPATCH_RigVMDispatch_ArrayIterator.Completed
   ) {is=struct RigVMUnitNode}
   > &RigUnit_TwoBoneIKSimplePerItem(
    DebugSettings=(bEnabled=False,Scale=10,WorldOffset=(Rotation=(X=0,Y=0,Z=0,W=1),Translation=(X=0,Y=0,Z=0),Scale3D=(X=1,Y=1,Z=1)))
    EffectorItem=(Type=Bone,Name="dyn_subbread02_03")
    ItemA=(Type=Bone,Name="dyn_subbread02_01")
    ItemALength=0
    ItemB=(Type=Bone,Name="dyn_subbread02_02")
    ItemBLength=0
    PoleVectorKind=Location
    PoleVectorSpace=(Type=Bone,Name="None")
    PrimaryAxis=(X=1,Y=0,Z=0)
    SecondaryAxis=(X=0,Y=-1,Z=0)
    SecondaryAxisWeight=0
    StretchMaximumRatio=1.25
    StretchStartRatio=0.75
    Weight=1
    bEnableStretch=false
    bPropagateToChildren=false
    Effector=&RigUnit_GetTransform(Item=(Type=Control,Name="dyn_subbread02_03_ctrl"), bInitial=False).Transform
    PoleVector=&RigUnit_GetTransform(Item=(Type=Control,Name="dyn_subbread02_02_ctrl"), bInitial=False).Transform.Translation
   )
   > &RigUnit_TwoBoneIKSimplePerItem(
    DebugSettings=(bEnabled=False,Scale=10,WorldOffset=(Rotation=(X=0,Y=0,Z=0,W=1),Translation=(X=0,Y=0,Z=0),Scale3D=(X=1,Y=1,Z=1)))
    EffectorItem=(Type=Bone,Name="dyn_subbread_upper01_03")
    ItemA=(Type=Bone,Name="dyn_subbread_upper01_01")
    ItemALength=0
    ItemB=(Type=Bone,Name="dyn_subbread_upper01_02")
    ItemBLength=0
    PoleVectorKind=Location
    PoleVectorSpace=(Type=Bone,Name="None")
    PrimaryAxis=(X=1,Y=0,Z=0)
    SecondaryAxis=(X=0,Y=-1,Z=0)
    SecondaryAxisWeight=0
    StretchMaximumRatio=1.25
    StretchStartRatio=0.75
    Weight=1
    bEnableStretch=false
    bPropagateToChildren=false
    PoleVector=&RigUnit_GetTransform(Item=(Type=Control,Name="dyn_subbread_upper01_02_ctrl"), bInitial=False).Transform.Translation
    Effector=&RigUnit_GetTransform(Item=(Type=Control,Name="dyn_subbread_upper01_03_ctrl"), bInitial=False).Transform
   )
   > &RigUnit_TwoBoneIKSimplePerItem(
    DebugSettings=(bEnabled=False,Scale=10,WorldOffset=(Rotation=(X=0,Y=0,Z=0,W=1),Translation=(X=0,Y=0,Z=0),Scale3D=(X=1,Y=1,Z=1)))
    EffectorItem=(Type=Bone,Name="dyn_subbread_upper02_03")
    ItemA=(Type=Bone,Name="dyn_subbread_upper02_01")
    ItemALength=0
    ItemB=(Type=Bone,Name="dyn_subbread_upper02_02")
    ItemBLength=0
    PoleVectorKind=Location
    PoleVectorSpace=(Type=Bone,Name="None")
    PrimaryAxis=(X=1,Y=0,Z=0)
    SecondaryAxis=(X=0,Y=-1,Z=0)
    SecondaryAxisWeight=1
    StretchMaximumRatio=1.25
    StretchStartRatio=0.75
    Weight=1
    bEnableStretch=false
    bPropagateToChildren=false
    PoleVector=&RigUnit_GetTransform(Item=(Type=Control,Name="dyn_subbread02_02_ctrl"), bInitial=False).Transform.Translation
    Effector=&RigUnit_GetTransform(Item=(Type=Control,Name="dyn_subbread_upper02_03_ctrl"), bInitial=False).Transform
   )
   > n2 = &RigUnit_TwoBoneIKSimplePerItem(
    DebugSettings=(bEnabled=False,Scale=10,WorldOffset=(Rotation=(X=0,Y=0,Z=0,W=1),Translation=(X=0,Y=0,Z=0),Scale3D=(X=1,Y=1,Z=1)))
    EffectorItem=(Type=Bone,Name="dyn_breads_side_03")
    ItemA=(Type=Bone,Name="dyn_breads_side_01")
    ItemALength=0
    ItemB=(Type=Bone,Name="dyn_breads_side_02")
    ItemBLength=0
    PoleVectorKind=Location
    PoleVectorSpace=(Type=Bone,Name="None")
    PrimaryAxis=(X=1,Y=0,Z=0)
    SecondaryAxis=(X=0,Y=-1,Z=0)
    SecondaryAxisWeight=1
    StretchMaximumRatio=1.25
    StretchStartRatio=0.75
    Weight=1
    bEnableStretch=false
    bPropagateToChildren=false
    PoleVector=&RigUnit_GetTransform(Item=(Type=Control,Name="dyn_breads_side_02_ctrl"), bInitial=False).Transform.Translation
    Effector=&RigUnit_GetTransform(Item=(Type=Control,Name="dyn_breads_side_03_ctrl"), bInitial=False).Transform
   )
   RigVMFunction_ControlFlowBranch = &RigVMFunction_ControlFlowBranch(ExecuteContext=@n2.ExecutePin, Condition=RigVMVariableNode() {variable=isSplineSmoothEnabled}.Value)
   RigUnit_ItemArray = &RigUnit_ItemArray(Items=((Type=Bone,Name="dyn_subbread_upper01_01"),(Type=Bone,Name="dyn_subbread_upper01_02"),(Type=Bone,Name="dyn_subbread_upper01_03")))
   SplineChainSmooth = RigVMFunctionReferenceNode(DebugThickness=false, ThreeJoints=@RigUnit_ItemArray.Items, ExecuteContext=@RigVMFunction_ControlFlowBranch.True)
   > SplineChainSmooth_1 = RigVMFunctionReferenceNode(
    DebugThickness=false
    ThreeJoints=&RigUnit_ItemArray(Items=((Type=Bone,Name="dyn_subbread_upper02_01"),(Type=Bone,Name="dyn_subbread_upper02_02"),(Type=Bone,Name="dyn_subbread_upper02_03"))).Items
   )
   > SplineChainSmooth_1_1 = RigVMFunctionReferenceNode(
    DebugThickness=false
    ThreeJoints=&RigUnit_ItemArray(Items=((Type=Bone,Name="dyn_breads_side_01"),(Type=Bone,Name="dyn_breads_side_02"),(Type=Bone,Name="dyn_breads_side_03"))).Items
   )
   > SplineChainSmooth_1_1_1 = RigVMFunctionReferenceNode(
    DebugThickness=false
    ThreeJoints=&RigUnit_ItemArray(Items=((Type=Bone,Name="dyn_subbread01_01"),(Type=Bone,Name="dyn_subbread01_02"),(Type=Bone,Name="dyn_subbread01_03"))).Items
   )
   > SplineChainSmooth_1_1_1_1 = RigVMFunctionReferenceNode(
    DebugThickness=false
    ThreeJoints=&RigUnit_ItemArray(Items=((Type=Bone,Name="dyn_subbread02_01"),(Type=Bone,Name="dyn_subbread02_02"),(Type=Bone,Name="dyn_subbread02_03"))).Items
   )
   RigUnit_BeginExecution = &RigUnit_BeginExecution()
  ]}
]
hierarchy [
 root {loc=<0 0 0>, rot=<0 0 0 1>}
 pack {parent=root, loc=<0 -0.7316 0>, rot=<0 0 0 1>}
 dyn_breads_upper_master {parent=pack, loc=<-0.2918 -3.1698 12.6466>, rot=<0.2621 -0.6918 0.2384 -0.6292>}
 dyn_breads_upper {parent=dyn_breads_upper_master, loc=<5.0456 -0.828 0>, rot=<0 0 0 1>}
 dyn_subbread_upper01_01 {parent=dyn_breads_upper, loc=<1.5576 -0.2775 -0.7528>, rot=<-0.0376 -0.0682 -0.1328 -0.9881>}
 dyn_subbread_upper01_02 {parent=dyn_subbread_upper01_01, loc=<5.3036 0 0>, rot=<0.1241 -0.0284 -0.1825 -0.9749>}
 dyn_subbread_upper01_03 {parent=dyn_subbread_upper01_02, loc=<6.1961 0 0>, rot=<-0.0435 0.0032 -0.0744 -0.9963>}
 dyn_subbread_upper02_01 {parent=dyn_breads_upper, loc=<1.4372 0.6614 0.8507>, rot=<-0.7231 -0.1127 -0.099 -0.6743>}
 dyn_subbread_upper02_02 {parent=dyn_subbread_upper02_01, loc=<5.2482 0 0>, rot=<0.7686 -0.1929 -0.0371 -0.6088>}
 dyn_subbread_upper02_03 {parent=dyn_subbread_upper02_02, loc=<5.4906 0 0>, rot=<-0.0475 0.0036 -0.0617 -0.997>}
 dyn_breads_side_master {parent=pack, loc=<-15.8735 -2.1978 -2.9149>, rot=<-0.078 -0.1279 0.5149 -0.844>}
 dyn_breads_side_01 {parent=dyn_breads_side_master, loc=<4.0734 0 0>, rot=<0.0585 -0.0078 -0.1833 -0.9813>}
 dyn_breads_side_02 {parent=dyn_breads_side_01, loc=<5.0707 0 0>, rot=<0.0672 0.0937 -0.3035 -0.9458>}
 dyn_breads_side_03 {parent=dyn_breads_side_02, loc=<4.9676 -0.0159 0.1529>, rot=<-0.0458 0.1688 -0.0219 -0.9843>}
 dyn_breadsmain {parent=pack, loc=<-4.5655 -10.1223 -6.1086>, rot=<-0.0109 -0.1306 0.0642 -0.9893>}
 dyn_breads_lower_master {parent=pack, loc=<-12.062 -1.3228 -16.3179>, rot=<-0.2571 0.3035 0.4969 -0.7713>}
 dyn_breads_lower {parent=dyn_breads_lower_master, loc=<3.0461 -0.2377 0>, rot=<0 0 -0.1981 -0.9802>}
 dyn_subbread01_01 {parent=dyn_breads_lower, loc=<1.2476 -0.1536 0.8313>, rot=<-0.0291 0.0492 -0.0876 -0.9945>}
 dyn_subbread01_02 {parent=dyn_subbread01_01, loc=<3.2591 0 0>, rot=<0.0461 0.048 -0.0332 -0.9972>}
 dyn_subbread01_03 {parent=dyn_subbread01_02, loc=<3.6672 0 0>, rot=<-0.1467 0.0114 -0.0933 -0.9847>}
 dyn_subbread02_01 {parent=dyn_breads_lower, loc=<1.3793 -0.0738 -0.9588>, rot=<-0.1443 0.0286 -0.1362 -0.9797>}
 dyn_subbread02_02 {parent=dyn_subbread02_01, loc=<3.3831 0 0>, rot=<0.0286 0.023 -0.048 -0.9982>}
 dyn_subbread02_03 {parent=dyn_subbread02_02, loc=<3.758 0 0>, rot=<-0.0368 0.0326 -0.1043 -0.9933>}
 dyn_breadsmain_ctrl {controlType=EULER_TRANSFORM, type=Control, loc=<0 0 0>, rot=<0 0 0 1>}
 dyn_subbread01_02_ctrl {controlType=EULER_TRANSFORM, parent=dyn_breadsmain_ctrl, type=Control, loc=<0 0 0>, rot=<0 0 0 1>}
 dyn_subbread02_02_ctrl {controlType=EULER_TRANSFORM, parent=dyn_breadsmain_ctrl, type=Control, loc=<0 0 0>, rot=<0 0 0 1>}
 dyn_subbread_upper01_02_ctrl {controlType=EULER_TRANSFORM, parent=dyn_breadsmain_ctrl, type=Control, loc=<-2.383 -11.3976 4.4426>, rot=<-0.0314 0.7982 -0.0237 0.6011>}
 dyn_subbread_upper02_02_ctrl {controlType=EULER_TRANSFORM, parent=dyn_breadsmain_ctrl, type=Control, loc=<0 0 0>, rot=<0 0 0 1>}
 dyn_breads_side_02_ctrl {controlType=EULER_TRANSFORM, parent=dyn_breadsmain_ctrl, type=Control, loc=<-9.9091 -9.6929 -7.2341>, rot=<0.0441 -0.0332 -0.0153 0.9984>}
 dyn_breads_side_03_ctrl {controlType=EULER_TRANSFORM, type=Control, loc=<0 0 0>, rot=<0 0 0 1>}
 dyn_subbread_upper02_03_ctrl {controlType=EULER_TRANSFORM, type=Control, loc=<0 0 0>, rot=<0 0 0 1>}
 dyn_subbread_upper01_03_ctrl {controlType=EULER_TRANSFORM, type=Control, loc=<0 0 0>, rot=<0 0 0 1>}
 dyn_subbread01_03_ctrl {controlType=EULER_TRANSFORM, type=Control, loc=<0 0 0>, rot=<0 0 0 1>}
 dyn_subbread02_03_ctrl {controlType=EULER_TRANSFORM, type=Control, loc=<0 0 0>, rot=<0 0 0 1>}
]
variables [isSplineSmoothEnabled {type=bool, default=True}]
