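# Physics asset for the DenseFog medium male backpack, setting up bodies and joints for its dangling "breads" bone chains around a large pack sphere.
# Upper and lower master capsules each lead into two three-segment sub-chains, while a side master leads into one, all linked by tightly limited swing joints with mostly locked twist.
# All five chain tips lock angularly to the central breadsmain sphere, which locks to the pack, forming closed loops; the masters have no constraint to the pack.
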PhysicsAsset /BRCosmetics/Accessories/FORT_Backpacks/Backpack_M_MED_DenseFog/Meshes/M_MED_DenseFog_Pack_Physics
bodies {
 dyn_breads_lower [{shapeType=Capsule, length=2.2032, radius=0.9213, shapeName=dyn_breads_lower_capsule}]
 dyn_breads_lower_master [{
   shapeType=Capsule
   length=1.4147
   radius=1.1143
   shapeName=dyn_breads_lower_master_capsule
   center=<1.5069 -0.1419 0.0189>
   rotation=<rot 89.2815 -179.9968 -179.9968>
  }]
 dyn_breads_side_01 [{
   shapeType=Capsule
   length=3.8811
   radius=1.1482
   shapeName=dyn_breads_side_01_capsule
   center=<2.4047 0.0002 0.0225>
   rotation=<rot -89.4623 3.1518 -3.1461>
  }]
 dyn_breads_side_02 [{
   shapeType=Capsule
   length=2.4276
   radius=1.1906
   shapeName=dyn_breads_side_02_capsule
   center=<1.9896 0 -0.208>
   rotation=<rot -84.033 179.9985 -179.9985>
  }]
 dyn_breads_side_03 [{shapeType=Capsule, length=0.7031, radius=0.9269, shapeName=dyn_breads_side_03_capsule}]
 dyn_breads_side_master [{
   shapeType=Capsule
   length=4.4958
   radius=0.9823
   shapeName=dyn_breads_side_master_capsule
   center=<2.3739 0 0.0626>
   rotation=<rot 88.4905 -179.9984 -179.9984>
  }]
 dyn_breads_upper [{shapeType=Capsule, length=1.4987, radius=1.0114, shapeName=dyn_breads_upper_capsule, center=<0 -0.0607 0>}]
 dyn_breads_upper_master [{
   shapeType=Capsule
   length=3.3603
   radius=1.3973
   shapeName=dyn_breads_upper_master_capsule
   center=<2.475 -0.1149 0.0259>
   rotation=<rot 89.4002 179.9731 179.9685>
  }]
 dyn_breadsmain [{shapeType=Sphere, radius=1.7141, shapeName=dyn_breadsmain_sphere, center=<0 -0.1418 0>}]
 dyn_subbread01_01 [{
   shapeType=Capsule
   length=3.7629
   radius=1.0048
   shapeName=dyn_subbread01_01_capsule
   center=<1.9499 0 0.0211>
   rotation=<rot 89.3799 -179.9922 -179.992>
  }]
 dyn_subbread01_02 [{
   shapeType=Capsule
   length=2.0324
   radius=1.0314
   shapeName=dyn_subbread01_02_capsule
   center=<1.2345 0 -0.0295>
   rotation=<rot -86.1824 179.9977 -179.998>
  }]
 dyn_subbread01_03 [{
   shapeType=Capsule
   length=2.5437
   radius=1.0366
   shapeName=dyn_subbread01_03_capsule
   center=<1.3433 -0.0017 0.0174>
   rotation=<rot -89.2578 0.6811 -0.7542>
  }]
 dyn_subbread02_01 [{
   shapeType=Capsule
   length=3.7629
   radius=1.0048
   shapeName=dyn_subbread02_01_capsule
   center=<1.9499 0 0.0211>
   rotation=<rot 89.3799 -179.9905 -179.9903>
  }]
 dyn_subbread02_02 [{
   shapeType=Capsule
   length=2.0324
   radius=1.0314
   shapeName=dyn_subbread02_02_capsule
   center=<1.2345 0 -0.0295>
   rotation=<rot -86.1824 179.9975 -179.9978>
  }]
 dyn_subbread02_03 [{
   shapeType=Capsule
   length=2.5437
   radius=1.0366
   shapeName=dyn_subbread02_03_capsule
   center=<1.338 0 -0.1206>
   rotation=<rot -84.8505 179.9987 -179.9989>
  }]
 dyn_subbread_upper01_01 [{
   shapeType=Capsule
   length=3.7036
   radius=1.0836
   shapeName=dyn_subbread_upper01_01_capsule
   center=<2.2252 0 0.0291>
   rotation=<rot 89.2517 179.9977 179.9976>
  }]
 dyn_subbread_upper01_02 [{
   shapeType=Capsule
   length=4.5107
   radius=1.0836
   shapeName=dyn_subbread_upper01_02_capsule
   center=<2.493 0 0.0326>
   rotation=<rot 89.2517 179.9965 179.9963>
  }]
 dyn_subbread_upper01_03 [{
   shapeType=Capsule
   length=2.0037
   radius=1.1057
   shapeName=dyn_subbread_upper01_03_capsule
   center=<0.9893 0 0.0218>
   rotation=<rot 88.736 179.9947 179.9943>
  }]
 dyn_subbread_upper02_01 [{
   shapeType=Capsule
   length=3.7036
   radius=1.0836
   shapeName=dyn_subbread_upper02_01_capsule
   center=<2.2252 -0.2066 0.0291>
   rotation=<rot 89.2517 179.9988 179.9987>
  }]
 dyn_subbread_upper02_02 [{
   shapeType=Capsule
   length=4.5107
   radius=1.0836
   shapeName=dyn_subbread_upper02_02_capsule
   center=<2.493 0 0.0326>
   rotation=<rot 89.2517 179.9958 179.9957>
  }]
 dyn_subbread_upper02_03 [{
   shapeType=Capsule
   length=2.0037
   radius=1.1057
   shapeName=dyn_subbread_upper02_03_capsule
   center=<0.9895 0 0.0026>
   rotation=<rot 89.8466 179.9794 179.9793>
  }]
 pack [{shapeType=Sphere, radius=14.7601, shapeName=pack_sphere, center=<0 5.5511 -1.8888>}]
}
constraints [
 {bone1Name=dyn_breads_upper, bone2Name=dyn_breads_upper_master, swing1LimitDegrees=2, swing2LimitDegrees=2, twistLimitDegrees=1}
 {
  bone1Name=dyn_subbread_upper01_01
  bone2Name=dyn_breads_upper
  swing1LimitDegrees=5
  swing2LimitDegrees=2
  twistLimitDegrees=45
  twistMotion=Locked
 }
 {
  bone1Name=dyn_subbread_upper02_01
  bone2Name=dyn_breads_upper
  swing1LimitDegrees=5
  swing2LimitDegrees=2
  twistLimitDegrees=45
  twistMotion=Locked
 }
 {
  bone1Name=dyn_subbread_upper02_02
  bone2Name=dyn_subbread_upper02_01
  swing1LimitDegrees=5
  swing2LimitDegrees=5
  twistLimitDegrees=45
  twistMotion=Locked
 }
 {
  bone1Name=dyn_subbread_upper01_02
  bone2Name=dyn_subbread_upper01_01
  swing1LimitDegrees=5
  swing2LimitDegrees=5
  twistLimitDegrees=45
  twistMotion=Locked
 }
 {
  bone1Name=dyn_subbread_upper02_03
  bone2Name=dyn_subbread_upper02_02
  swing1LimitDegrees=5
  swing2LimitDegrees=7
  twistLimitDegrees=45
  twistMotion=Locked
 }
 {
  bone1Name=dyn_subbread_upper01_03
  bone2Name=dyn_subbread_upper01_02
  swing1LimitDegrees=5
  swing2LimitDegrees=7
  twistLimitDegrees=45
  twistMotion=Locked
 }
 {
  bone1Name=dyn_subbread_upper02_03
  bone2Name=dyn_breadsmain
  swing1LimitDegrees=45
  swing1Motion=Locked
  swing2LimitDegrees=45
  swing2Motion=Locked
  twistLimitDegrees=45
  twistMotion=Locked
 }
 {
  bone1Name=dyn_subbread_upper01_03
  bone2Name=dyn_breadsmain
  swing1LimitDegrees=45
  swing1Motion=Locked
  swing2LimitDegrees=45
  swing2Motion=Locked
  twistLimitDegrees=45
  twistMotion=Locked
 }
 {
  bone1Name=dyn_subbread02_03
  bone2Name=dyn_breadsmain
  swing1LimitDegrees=45
  swing1Motion=Locked
  swing2LimitDegrees=45
  swing2Motion=Locked
  twistLimitDegrees=45
  twistMotion=Locked
 }
 {
  bone1Name=dyn_subbread01_03
  bone2Name=dyn_breadsmain
  swing1LimitDegrees=45
  swing1Motion=Locked
  swing2LimitDegrees=45
  swing2Motion=Locked
  twistLimitDegrees=45
  twistMotion=Locked
 }
 {bone1Name=dyn_breads_lower, bone2Name=dyn_breads_lower_master, swing1LimitDegrees=2, swing2LimitDegrees=2, twistLimitDegrees=2}
 {
  bone1Name=dyn_subbread01_01
  bone2Name=dyn_breads_lower
  swing1LimitDegrees=5
  swing2LimitDegrees=3
  twistLimitDegrees=45
  twistMotion=Locked
 }
 {
  bone1Name=dyn_subbread02_01
  bone2Name=dyn_breads_lower
  swing1LimitDegrees=5
  swing2LimitDegrees=3
  twistLimitDegrees=45
  twistMotion=Locked
 }
 {
  bone1Name=dyn_subbread01_02
  bone2Name=dyn_subbread01_01
  swing1LimitDegrees=5
  swing2LimitDegrees=5
  twistLimitDegrees=45
  twistMotion=Locked
 }
 {
  bone1Name=dyn_subbread02_02
  bone2Name=dyn_subbread02_01
  swing1LimitDegrees=5
  swing2LimitDegrees=5
  twistLimitDegrees=45
  twistMotion=Locked
 }
 {
  bone1Name=dyn_subbread01_03
  bone2Name=dyn_subbread01_02
  swing1LimitDegrees=5
  swing2LimitDegrees=45
  swing2Motion=Locked
  twistLimitDegrees=45
  twistMotion=Locked
 }
 {
  bone1Name=dyn_subbread02_03
  bone2Name=dyn_subbread02_02
  swing1LimitDegrees=5
  swing2LimitDegrees=45
  swing2Motion=Locked
  twistLimitDegrees=45
  twistMotion=Locked
 }
 {bone1Name=dyn_breads_side_01, bone2Name=dyn_breads_side_master, swing1LimitDegrees=5, swing2LimitDegrees=10, twistLimitDegrees=2}
 {
  bone1Name=dyn_breads_side_02
  bone2Name=dyn_breads_side_01
  swing1LimitDegrees=5
  swing2LimitDegrees=15
  twistLimitDegrees=45
  twistMotion=Locked
 }
 {
  bone1Name=dyn_breads_side_03
  bone2Name=dyn_breads_side_02
  swing1LimitDegrees=45
  swing1Motion=Locked
  swing2LimitDegrees=10
  twistLimitDegrees=45
  twistMotion=Locked
 }
 {
  bone1Name=dyn_breads_side_03
  bone2Name=dyn_breadsmain
  swing1LimitDegrees=45
  swing1Motion=Locked
  swing2LimitDegrees=45
  swing2Motion=Locked
  twistLimitDegrees=45
  twistMotion=Locked
 }
 {
  bone1Name=dyn_breadsmain
  bone2Name=pack
  swing1LimitDegrees=10
  swing1Motion=Locked
  swing2LimitDegrees=10
  swing2Motion=Locked
  twistLimitDegrees=10
  twistMotion=Locked
 }
]
