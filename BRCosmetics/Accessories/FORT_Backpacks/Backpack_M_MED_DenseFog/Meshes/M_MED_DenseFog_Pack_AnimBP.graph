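# Animation Blueprint for the DenseFog backpack cosmetic that offsets the pack against body movement and adds secondary physics.
# Each update turns reference-pose deltas of spine and upper-arm bones into range-mapped corrections, mostly interpolated.
# Three additive Modify Bone nodes apply these to `pack`, then a RigidBody node simulates the pack's physics asset.
# Anim Dynamics, Control Rig and clavicle math are unconnected; skydive/crouch alphas are set but unread, and skydive offsets stay zero.
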AnimBlueprint /BRCosmetics/Accessories/FORT_Backpacks/Backpack_M_MED_DenseFog/Meshes/M_MED_DenseFog_Pack_AnimBP
target_skeleton=&/M_MED_DenseFog_Pack_Skeleton
graphs [
 AnimGraph {nodes [
   n1 = Anim Dynamics - Bone: dyn_breadsmain!(ComponentPose, bAlphaBoolEnabled:bool=True, ExternalForce="0,0,0") -> (Pose) {
    alphaInputType=Bool
    angularDampingOverride=1
    angularSpringConstant=0.5
    bGravityOverrideInSimSpace=true
    bLinearSpring=true
    bOverrideAngularDamping=true
    bOverrideLinearDamping=true
    bUseGravityOverride=true
    linearDampingOverride=1
    linearSpringConstant=10
    alphaBoolBlend {blendOption=Linear}
    boundBone {boneName=dyn_breadsmain}
    componentAppliedLinearAccClamp=<500 350 350>
    componentLinearAccScale=<2 2 2>
    componentLinearVelScale=<2 2 2>
    gravityOverride {z=-980}
    physicsBodyDefinitions [{
      boundBone {boneName=dyn_breadsmain}
      boxExtents=<5 5 5>
      constraintSetup {angularLimitsMax=<0 0 0>, angularLimitsMin=<0 0 0>, angularTarget=<1 0 0>, linearAxesMax=<0.7 0 1>, linearAxesMin=<-1.2 0 -1>}
      localJointOffset=<0 0 0.1>
     }]
   }
   n2 = Anim Dynamics - Bone: dyn_breads_upper_master!(ComponentPose=@n1, bAlphaBoolEnabled:bool=True, ExternalForce="0,0,0") -> (Pose) {
    alphaInputType=Bool
    angularDampingOverride=0.9999
    bOverrideAngularDamping=true
    bOverrideLinearDamping=true
    bUseGravityOverride=true
    linearDampingOverride=0.9999
    alphaBoolBlend {blendOption=Linear}
    boundBone {boneName=dyn_breads_upper_master}
    componentAppliedLinearAccClamp=<2000 2000 2000>
    componentLinearAccScale=<0.1 0.1 0.1>
    componentLinearVelScale=<0.1 0.1 0.1>
    gravityOverride {z=-980}
    physicsBodyDefinitions [{
      boundBone {boneName=dyn_breads_upper_master}
      boxExtents=<5 2 2>
      constraintSetup {angularLimitsMax=<0 0 1>, angularLimitsMin=<0 0 -1>, angularTarget=<1 0 0>}
      localJointOffset=<0 0 0.1>
     }]
   }
   n3 = Anim Dynamics - Bone: dyn_breads_upper!(ComponentPose=@n2, bAlphaBoolEnabled:bool=True, ExternalForce="0,0,0") -> (Pose) {
    alphaInputType=Bool
    angularDampingOverride=0.9999
    bOverrideAngularDamping=true
    bOverrideLinearDamping=true
    bUseGravityOverride=true
    linearDampingOverride=0.9999
    alphaBoolBlend {blendOption=Linear}
    boundBone {boneName=dyn_breads_upper}
    componentAppliedLinearAccClamp=<2000 2000 2000>
    componentLinearAccScale=<0.1 0.1 0.1>
    componentLinearVelScale=<0.1 0.1 0.1>
    gravityOverride {z=-980}
    physicsBodyDefinitions [{
      boundBone {boneName=dyn_breads_upper}
      boxExtents=<5 2 2>
      constraintSetup {angularLimitsMax=<0 0 2>, angularLimitsMin=<0 0 -2>, angularTarget=<1 0 0>}
      localJointOffset=<0 0 0.1>
     }]
   }
   n4 = Anim Dynamics - Bone: dyn_breads_lower_master!(ComponentPose=@n3, bAlphaBoolEnabled:bool=True, ExternalForce="0,0,0") -> (Pose) {
    alphaInputType=Bool
    angularDampingOverride=0.9999
    bOverrideAngularDamping=true
    bOverrideLinearDamping=true
    bUseGravityOverride=true
    linearDampingOverride=0.9999
    alphaBoolBlend {blendOption=Linear}
    boundBone {boneName=dyn_breads_lower_master}
    componentAppliedLinearAccClamp=<2000 2000 2000>
    componentLinearAccScale=<0.1 0.1 0.1>
    componentLinearVelScale=<0.1 0.1 0.1>
    gravityOverride {z=-980}
    physicsBodyDefinitions [{
      boundBone {boneName=dyn_breads_lower_master}
      boxExtents=<5 2 2>
      constraintSetup {angularLimitsMax=<0 0 2>, angularLimitsMin=<0 0 -5>, angularTarget=<1 0 0>}
      localJointOffset=<0 0 0.1>
     }]
   }
   n5 = Anim Dynamics - Bone: dyn_breads_lower_master(ComponentPose=@n4, bAlphaBoolEnabled=True, ExternalForce="0,0,0") {
    alphaInputType=Bool
    angularDampingOverride=0.9999
    bOverrideAngularDamping=true
    bOverrideLinearDamping=true
    bUseGravityOverride=true
    linearDampingOverride=0.9999
    alphaBoolBlend {blendOption=Linear}
    boundBone {boneName=dyn_breads_lower_master}
    componentAppliedLinearAccClamp=<2000 2000 2000>
    componentLinearAccScale=<0.1 0.1 0.1>
    componentLinearVelScale=<0.1 0.1 0.1>
    gravityOverride {z=-980}
    physicsBodyDefinitions [{
      boundBone {boneName=dyn_breads_lower_master}
      boxExtents=<5 2 2>
      constraintSetup {angularLimitsMax=<0 0 2>, angularLimitsMin=<0 0 -5>, angularTarget=<1 0 0>}
      localJointOffset=<0 0 0.1>
     }]
   }
   Anim Dynamics - Bone: dyn_breads_side_master!(ComponentPose=@n5, bAlphaBoolEnabled:bool=True, ExternalForce="0,0,0") -> (Pose) {
    alphaInputType=Bool
    angularDampingOverride=0.9999
    bOverrideAngularDamping=true
    bOverrideLinearDamping=true
    bUseGravityOverride=true
    linearDampingOverride=0.9999
    alphaBoolBlend {blendOption=Linear}
    boundBone {boneName=dyn_breads_side_master}
    componentAppliedLinearAccClamp=<2000 2000 2000>
    componentLinearAccScale=<0.1 0.1 0.1>
    componentLinearVelScale=<0.1 0.1 0.1>
    gravityOverride {z=-980}
    physicsBodyDefinitions [{
      boundBone {boneName=dyn_breads_side_master}
      boxExtents=<5 2 2>
      constraintSetup {angularLimitsMax=<0 0 1>, angularLimitsMin=<0 0 -1>, angularTarget=<1 0 0>}
      localJointOffset=<0 0 0.1>
     }]
   }
   Control Rig!(Source, bAlphaBoolEnabled:bool=True, isSplineSmoothEnabled:bool=true) -> (Pose) {
    alphaInputType=Bool
    alphaBoolBlend {blendOption=Linear}
    controlRigAssetReference {blueprintRigClass=&/M_MED_DenseFog_Pack_CtrlRig.M_MED_DenseFog_Pack_CtrlRig_C}
   }
   n6 = "Transform (Modify) Bone - Bone: pack"(
    ComponentPose=Local To Component(LocalPose=Input Pose())
    Translation=Get Backpack Translation Skydive() -> (Backpack Translation Skydive:Vector)
    Rotation=Get Backpack Rotate Skydive() -> (Backpack Rotate Skydive:Rotator)
   ) {rotationMode=BMM_Additive, translationMode=BMM_Additive, alphaBoolBlend {blendOption=Linear}, boneToModify {boneName=pack}}
   n7 = "Transform (Modify) Bone - Bone: pack"(
    ComponentPose=@n6
    Translation=Get SpineRotationTranslationCorr() -> (SpineRotationTranslationCorr:Vector)
    Rotation=Get SpineRotationCorr() -> (SpineRotationCorr:Rotator)
   ) {rotationMode=BMM_Additive, translationMode=BMM_Additive, alphaBoolBlend {blendOption=Linear}, boneToModify {boneName=pack}}
   n8 = "Transform (Modify) Bone - Bone: pack"(
    ComponentPose=@n7
    Translation=Get ArmBackSideTranslationCorr() -> (ArmBackSideTranslationCorr:Vector)
    Rotation=Get Backpack Rotate Skyfall() -> (Backpack Rotate Skyfall:Rotator)
   ) {rotationMode=BMM_Additive, translationMode=BMM_Additive, alphaBoolBlend {blendOption=Linear}, boneToModify {boneName=pack}}
   n9 = RigidBody(ComponentPose=@n8) {
    bOverrideWorldGravity=true
    overridePhysicsAsset=&/M_MED_DenseFog_Pack_Physics
    alphaBoolBlend {blendOption=Linear}
    componentAppliedLinearAccClamp=<2000 2000 2000>
    componentLinearAccScale=<0.1 0.1 0.1>
    componentLinearVelScale=<0.1 0.1 0.1>
    overrideWorldGravity {z=-980}
    simSpaceSettings {maxAngularAcceleration=500, maxAngularVelocity=10, maxLinearAcceleration=350, maxLinearVelocity=100, worldAlpha=0.5}
   }
   Output Pose(Result=Component To Local(ComponentPose=@n9)) {layerGroup=None}
  ]}
 EventGraph {nodes [
   n1 = Event BlueprintUpdateAnimation()
   > Set Backpack Rotate Skydive(Backpack Rotate Skydive:Rotator="0, 0, 0") -> (Output_Get:Rotator)
   > Set Backpack Translation Skydive(Backpack Translation Skydive:Vector="0, 0, 0") -> (Output_Get:Vector)
   n2 = RelativeTransformFromParent!(BoneCalculated:Name=spine_05, BoneRelativeParent:Name=spine_01) -> (X Movement:double, Y Movement:double, Z Movement:double, X Rotation:double, Y Rotation:double, Z Rotation:double)
   n3 = MapRangeClamped(Value=@n2.X Rotation, InRangeA=0, InRangeB=25, OutRangeB=-5)
   n4 = Get!() -> (SpineRotationCorr_Roll:float, SpineRotationCorr_Pitch:float, SpineRotationCorr_Yaw:float)
   > Set SpineRotationCorr(SpineRotationCorr:Rotator=MakeRotator(
     Roll=MapRangeClamped(Value=@n2.Z Rotation, InRangeA=-45, InRangeB=0, OutRangeA=10, OutRangeB=0)
     Yaw=FInterpTo(Current=@n4.SpineRotationCorr_Yaw, Target=@n3, DeltaTime=@n1.DeltaTimeX, InterpSpeed=6)
    )) -> (Output_Get:Rotator)
   n5 = float + float(
    A=MapRangeClamped(Value=@n2.Z Rotation, InRangeA=-30, InRangeB=0, OutRangeA=-4, OutRangeB=0)
    B=MapRangeClamped(Value=@n2.X Rotation, InRangeA=-25, InRangeB=25, OutRangeA=-0.2, OutRangeB=-0.2)
   )
   n6 = Get!() -> (SpineRotationTranslationCorr_X:double, SpineRotationTranslationCorr_Y:double, SpineRotationTranslationCorr_Z:double)
   > Set SpineRotationTranslationCorr(SpineRotationTranslationCorr:Vector=MakeVector(Y=FInterpTo(Current=@n6.SpineRotationTranslationCorr_Y, Target=@n5, DeltaTime=@n1.DeltaTimeX, InterpSpeed=10))) -> (Output_Get:Vector)
   n7 = RelativeTransformFromParent(BoneCalculated=upperarm_r, BoneRelativeParent=spine_05)
   n8 = RelativeTransformFromParent(BoneCalculated=upperarm_l, BoneRelativeParent=spine_05)
   n9 = MapRangeClamped(
    Value=float * float(A=float + float(A=@n8.Z Rotation, B=@n7.Z Rotation), B=0.5) {is=title K2Node_PromotableOperator}
    InRangeA=-38
    InRangeB=0
    OutRangeA=-3
    OutRangeB=0
   )
   n10 = Get!() -> (Backpack Translation Skydive_X:double, Backpack Translation Skydive_Y:double, Backpack Translation Skydive_Z:double)
   > Set ArmBackSideTranslationCorr(ArmBackSideTranslationCorr:Vector=MakeVector(Y=FInterpTo(Current=@n10.Backpack Translation Skydive_Y, Target=@n9, DeltaTime=@n1.DeltaTimeX, InterpSpeed=3))) -> (Output_Get:Vector)
   Event OnEndSkydiving()
   > Set SkydiveAlpha(SkydiveAlpha:double=0) -> (Output_Get:double)
   Event OnBeginSkydiving()
   > Set SkydiveAlpha(SkydiveAlpha=1)
   Event OnBeginCrouching()
   > Set CrouchAlpha(CrouchAlpha:double=1) -> (Output_Get:double)
   Event OnEndCrouching()
   > Set CrouchAlpha(CrouchAlpha=0)
   n11 = RelativeTransformFromParent(BoneCalculated=clavicle_l, BoneRelativeParent=spine_05)
   n12 = RelativeTransformFromParent(BoneCalculated=clavicle_r, BoneRelativeParent=spine_05)
   MakeRotator(Yaw=float + float(
     A=MapRangeClamped(Value=@n11.Z Rotation, InRangeA=-5, InRangeB=-30, OutRangeB=-20)
     B=MapRangeClamped(Value=@n12.Z Rotation, InRangeA=-5, InRangeB=-30, OutRangeB=20)
    ))
   n13 = RelativeTransformFromParent(BoneCalculated=clavicle_l, BoneRelativeParent=spine_05)
   n14 = MapRangeClamped(Value=@n13.Z Rotation, InRangeA=-5, InRangeB=-30, OutRangeB=2)
   n15 = RelativeTransformFromParent(BoneCalculated=clavicle_r, BoneRelativeParent=spine_05)
   n16 = MapRangeClamped(Value=@n15.Z Rotation, InRangeA=-5, InRangeB=-30, OutRangeB=-2)
   MakeVector(
    X=float + float(
     A=MapRangeClamped(Value=@n13.Z Rotation, InRangeA=-5, InRangeB=-30, OutRangeB=-3)
     B=MapRangeClamped(Value=@n15.Z Rotation, InRangeA=-5, InRangeB=-30, OutRangeB=3)
    )
    Y="Clamp (Float)"(Value=float + float(A=@n14, B=@n16), Min=-2, Max=2)
   )
  ]}
 RelativeTransformFromParent {nodes [
   n1 = RelativeTransformFromParent!() -> (then, BoneCalculated:Name, BoneRelativeParent:Name) {is=title K2Node_FunctionEntry}
   n2 = BreakTransform!(InTransform=GetDeltaTransformFromRefPose(self=GetOwnerSkeletalMeshComponent(), BoneName=@n1, BaseName=@n1.BoneRelativeParent)) -> (Location_X:double, Location_Y:double, Location_Z:double, Rotation_Roll:float, Rotation_Pitch:float, Rotation_Yaw:float, Scale)
   > Return Node!(
    execute
    X Movement:double=@n2
    Y Movement:double=@n2.Location_Y
    Z Movement:double=@n2.Location_Z
    X Rotation:double=@n2.Rotation_Roll
    Y Rotation:double=@n2.Rotation_Pitch
    Z Rotation:double=@n2.Rotation_Yaw
   )
  ]}
]
variables [
 ArmBackSideTranslationCorr {type=Vector}
 Backpack Rotate Skydive {type=Rotator}
 Backpack Rotate Skyfall {type=Rotator}
 Backpack Translation Skydive {type=Vector}
 CrouchAlpha {type=number}
 IsLODGreaterThanZero {type=boolean}
 SkydiveAlpha {type=number}
 SpineRotationCorr {type=Rotator}
 SpineRotationTranslationCorr {type=Vector}
]
